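# Niagara swing-trail effect for the Stars pickaxe cosmetic, drawing hue-shifted purple ribbons between two bones or sockets as the weapon animates.
# MainWireframeTrail draws the central ribbon, while OutlineTrail spawns two thinner trails offset from the bone positions.
# Both fade alpha over short particle lifetimes and shape width with curves indexed by spawn time within the swing and a system-level swing fade.
# The outline offset scales by those same curves.
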NiagaraSystem /BRCosmetics/Weapons/FORT_Melee/Pickaxe_Stars/FX/NS_Pickaxe_Stars_AnimTrail
dependencies {
 usedDataInterfaces [&NiagaraDataInterfaceCurve, &NiagaraDataInterfaceSocketReader]
 usedDynamicInputs []
 usedModules [
  &self:SystemSpawnScript.SystemScriptSource.SystemScriptGraph.NiagaraNodeAssignment_0.SetVariables_#1
  &SystemState
  &self:SystemSpawnScript.SystemScriptSource.SystemScriptGraph.NiagaraNodeAssignment_1.SetVariables_#2
 ]
}
emitters {
 MainWireframeTrail {
  emitterUpdateScript [
   &EmitterState()
   &FN_SpawnAnimTrail()
  ]
  particleSpawnScript [
   &FN_InitializeAnimTrailBoneIndecies(First Bone or Socket Index=1, Second Bone or Socket Index=0)
   &FN_InitializeAnimTrail(AnimTrailWidth=2.1, Lifetime=0.3)
  ]
  particleUpdateScript [
   &FN_UpdateAnimTrail()
   /SetVariables(Particles.NormalizedIndexFixed:NiagaraFloat=&ClampFloat(Float=&Subtract_Float(A=$Emitter.Age, B=$Particles.Age))) {isSetParameters=true}
   /SetVariables(Particles.RibbonU0Override=&MakeCustomFloatFromBool(
     Bool=&SetBoolByIntComparison(A=$Engine.ExecIndex, B=0, Comparison Type=::NewEnumerator4)
     True Float=&MakeFloatFromInt(Int=&Multiply_Int(A=$Engine.Emitter.NumParticles, B=2))
     False Float=&Multiply_FloatByInt(Float=2, Integer=$Engine.ExecIndex)
    )) {isSetParameters=true}
   &ScaleColor(
    Scale Alpha=&FloatFromCurve(FloatCurve=DataInterface() {propertyValues {Curve {keys [
         {arriveTangent=-17.1094, interpMode=RCIM_Cubic, leaveTangent=-17.1094, tangentMode=RCTM_User, time=0.005, value=3}
         {arriveTangent=-0.4835, interpMode=RCIM_Cubic, leaveTangent=-0.4835, tangentMode=RCTM_User, time=0.2877, value=0.1515}
         {arriveTangent=-0.0191, interpMode=RCIM_Cubic, leaveTangent=-0.0191, tangentMode=RCTM_User, time=1.0011, value=0}
        ]}}})
    Scale RGB=&VectorFromFloat(Value=1)
    ScaleA=true
    ScaleRGB=true
   )
   /Game/Effects/Niagara/Modules/Color/FN_ColorHueShift(
    Color:LinearColor=<rgba 0.3961 0.0902 1 1>
    Initial Offset:NiagaraFloat=&Add_Float(
     A=0
     B=&Multiply_Float(A=&OneMinusFloat(Float=&ReturnNormalizedExecIndex()), B=1)
    )
   )
   &ScaleRibbonWidth(Ribbon Width Scale=&FloatFromCurve(
     FloatCurve=DataInterface() {propertyValues {Curve {keys [
         {arriveTangent=11.4573, interpMode=RCIM_Cubic, leaveTangent=11.4573, tangentMode=RCTM_User, time=0, value=0}
         {arriveTangentWeight=0.05, interpMode=RCIM_Cubic, leaveTangentWeight=0.2501, tangentMode=RCTM_User, time=0.25, value=1}
         {arriveTangent=-1.5611, interpMode=RCIM_Cubic, leaveTangent=-1.5611, tangentMode=RCTM_User, time=1, value=0}
        ]}}}
     Scale Curve=&FloatFromCurve(
      FloatCurve=DataInterface() {propertyValues {Curve {keys [
          {arriveTangent=22.2626, interpMode=RCIM_Cubic, leaveTangent=22.2626, tangentMode=RCTM_User, time=0, value=0}
          {interpMode=RCIM_Cubic, leaveTangentWeight=0.0667, time=0.12, value=1}
          {arriveTangent=-23.2388, interpMode=RCIM_Cubic, leaveTangent=-23.2388, tangentMode=RCTM_User, time=0.22, value=0}
         ]}}}
      CurveIndex=$Particles.NormalizedIndexFixed
     )
    ))
   &ScaleRibbonWidth(Ribbon Width Scale=$System.SwingGeneralFade) {module=ScaleRibbonWidth001}
   &DynamicMaterialParameters(
    Index 1 Param 3=0.5
    Index 1 Param 4=0.5
    Module.Index0Float0Write=true
    Module.Index0Float1Write=true
    Module.Index0Float2Write=true
    Module.Index0Float3Write=true
    Module.Index1Float0Write=true
    Module.Index1Float1Write=true
    Module.Index1Float2Write=true
    Module.Index1Float3Write=true
    Write Parameter Index 1=::NewEnumerator1
   )
  ]
  rendererClasses [&NiagaraRibbonRendererProperties]
  renderers [{rendererClass=&NiagaraRibbonRendererProperties}]
 }
 OutlineTrail {
  emitterUpdateScript [
   &EmitterState()
   &FN_SpawnAnimTrail(Number of Anim Trails=2)
  ]
  particleSpawnScript [
   &FN_InitializeAnimTrailBoneIndecies(First Bone or Socket Index=1, Second Bone or Socket Index=0)
   &FN_InitializeAnimTrailBoneIndecies(AnimTrail Selection=1, First Bone or Socket Index=1, Second Bone or Socket Index=0) {module=FN_InitializeAnimTrailBoneIndecies001}
   &FN_InitializeAnimTrail(AnimTrailWidth=0.84, Lifetime=0.3)
   &self:OffsetAnimtrail(
    Offset:NiagaraFloat=&MaskFloat(Mask Source=0, Mask Parameter=?, Mask=0, Masked Float=1, Passthrough Float=-1)
    firstBonePos:Vector3f=$Particles.FirstBonePosition
    secondBonePos:Vector3f=$Particles.SecondBonePosition
   ) {module=OffsetAnimtrail}
  ]
  particleUpdateScript [
   &FN_UpdateAnimTrail()
   /SetVariables(Particles.NormalizedIndexFixed:NiagaraFloat=&ClampFloat(Float=&Subtract_Float(A=$Emitter.Age, B=$Particles.Age))) {isSetParameters=true}
   &ScaleColor(
    Scale Alpha=&FloatFromCurve(FloatCurve=DataInterface() {propertyValues {Curve {keys [
         {arriveTangent=-17.1094, interpMode=RCIM_Cubic, leaveTangent=-17.1094, tangentMode=RCTM_User, time=0.005, value=3}
         {arriveTangent=-0.4835, interpMode=RCIM_Cubic, leaveTangent=-0.4835, tangentMode=RCTM_User, time=0.2877, value=0.1515}
         {arriveTangent=-0.0191, interpMode=RCIM_Cubic, leaveTangent=-0.0191, tangentMode=RCTM_User, time=1.0011, value=0}
        ]}}})
    Scale RGB=&VectorFromFloat(Value=1)
    ScaleA=true
    ScaleRGB=true
   )
   &FN_ColorHueShift(
    Color:LinearColor=<rgba 0.3961 0.0902 1 1>
    Initial Offset:NiagaraFloat=&Add_Float(
     A=0
     B=&Multiply_Float(A=&OneMinusFloat(Float=&ReturnNormalizedExecIndex()), B=1)
    )
   )
   &ScaleRibbonWidth(Ribbon Width Scale=&FloatFromCurve(
     FloatCurve=DataInterface() {propertyValues {Curve {keys [
         {arriveTangent=11.4573, interpMode=RCIM_Cubic, leaveTangent=11.4573, tangentMode=RCTM_User, time=0, value=0}
         {arriveTangentWeight=0.05, interpMode=RCIM_Cubic, leaveTangentWeight=0.2501, tangentMode=RCTM_User, time=0.25, value=1}
         {arriveTangent=-1.5611, interpMode=RCIM_Cubic, leaveTangent=-1.5611, tangentMode=RCTM_User, time=1, value=0}
        ]}}}
     Scale Curve=&FloatFromCurve(
      FloatCurve=DataInterface() {propertyValues {Curve {keys [
          {arriveTangent=22.2626, interpMode=RCIM_Cubic, leaveTangent=22.2626, tangentMode=RCTM_User, time=0, value=0}
          {interpMode=RCIM_Cubic, leaveTangentWeight=0.0667, time=0.12, value=1}
          {arriveTangent=-23.2388, interpMode=RCIM_Cubic, leaveTangent=-23.2388, tangentMode=RCTM_User, time=0.22, value=0}
         ]}}}
      CurveIndex=$Particles.NormalizedIndexFixed
     )
    ))
   &ScaleRibbonWidth(Ribbon Width Scale=$System.SwingGeneralFade) {module=ScaleRibbonWidth001}
   &DynamicMaterialParameters(
    Index 1 Param 3=0.5
    Index 1 Param 4=0.5
    Module.Index0Float0Write=true
    Module.Index0Float1Write=true
    Module.Index0Float2Write=true
    Module.Index0Float3Write=true
    Module.Index1Float0Write=true
    Module.Index1Float1Write=true
    Module.Index1Float2Write=true
    Module.Index1Float3Write=true
    Write Parameter Index 1=::NewEnumerator1
   )
   &self:ApplyOffset(Scale:NiagaraFloat=&Multiply_Float(
     A=&FloatFromCurve(
      FloatCurve=DataInterface() {propertyValues {Curve {keys [
          {arriveTangent=11.4573, interpMode=RCIM_Cubic, leaveTangent=11.4573, tangentMode=RCTM_User, time=0, value=0}
          {arriveTangentWeight=0.05, interpMode=RCIM_Cubic, leaveTangentWeight=0.2501, tangentMode=RCTM_User, time=0.25, value=1}
          {arriveTangent=-1.5611, interpMode=RCIM_Cubic, leaveTangent=-1.5611, tangentMode=RCTM_User, time=1, value=0}
         ]}}}
      Scale Curve=&FloatFromCurve(
       FloatCurve=DataInterface() {propertyValues {Curve {keys [
           {arriveTangent=22.2626, interpMode=RCIM_Cubic, leaveTangent=22.2626, tangentMode=RCTM_User, time=0, value=0}
           {interpMode=RCIM_Cubic, leaveTangentWeight=0.0667, time=0.12, value=1}
           {arriveTangent=-23.2388, interpMode=RCIM_Cubic, leaveTangent=-23.2388, tangentMode=RCTM_User, time=0.22, value=0}
          ]}}}
       CurveIndex=$Particles.NormalizedIndexFixed
      )
     )
     B=$System.SwingGeneralFade
    )) {module=ApplyOffset}
  ]
  rendererClasses [&NiagaraRibbonRendererProperties]
  renderers [{rendererClass=&NiagaraRibbonRendererProperties}]
 }
}
summary {
 systemName=NS_Pickaxe_Stars_AnimTrail
 emitters [
  {emitterName=MainWireframeTrail, rendererClasses [&NiagaraRibbonRendererProperties]}
  {emitterName=OutlineTrail, rendererClasses [&NiagaraRibbonRendererProperties]}
 ]
 userVariables [
  User.bIsFrontEndPreview {type=NiagaraBool}
  User.RibbonIDOffset {type=NiagaraInt32}
 ]
}
